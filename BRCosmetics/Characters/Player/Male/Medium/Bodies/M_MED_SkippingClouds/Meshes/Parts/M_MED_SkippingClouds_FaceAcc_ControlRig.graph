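# Control Rig for the face accessory part of the SkippingClouds medium male player body, applying corrective offsets to hair bones as the head moves.
# On forward solve, a spherical pose reader measures head orientation relative to spine_05 within a height-stretched active and falloff region.
# Its output weights a Modify Transforms unit that offsets dyn_hair bones such as dyn_hair_15 and dyn_hair_16, a pose-driven fix without simulation.
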ControlRigBlueprint /BRCosmetics/Characters/Player/Male/Medium/Bodies/M_MED_SkippingClouds/Meshes/Parts/M_MED_SkippingClouds_FaceAcc_ControlRig
graphs [
 RigVMFunctionLibrary {links [], nodes []}
 RigVMModel {nodes [
   RigUnit_BeginExecution = &RigUnit_BeginExecution()
   n1 = &RigUnit_SphericalPoseReader(
    ActiveRegionScaleFactors=(PositiveWidth=1,NegativeWidth=1,PositiveHeight=4.5,NegativeHeight=4.5)
    ActiveRegionSize=0.05
    Debug=(bDrawDebug=false,bDraw2D=False,bDrawLocalAxes=False,DebugScale=25,DebugSegments=20,DebugThickness=0.25)
    DriverAxis=(X=0,Y=-1,Z=0)
    DriverItem=(Type=Bone,Name="head")
    FalloffRegionScaleFactors=(PositiveWidth=1,NegativeWidth=1,PositiveHeight=6,NegativeHeight=6)
    FalloffSize=0.3
    FlipHeightScaling=False
    FlipWidthScaling=False
    OptionalParentItem=(Type=Bone,Name="spine_05")
    RotationOffset=(X=-90,Y=0,Z=-50)
    ExecutePin=@RigUnit_BeginExecution.ExecutePin
   )
   > &RigUnit_ModifyTransforms(
    ItemToModify="((Item=(Type=Bone,Name=\"dyn_hair_16\"),Transform=(Rotation=(X=0,Y=0.5,Z=0,W=0.866),Translation=(X=0,Y=0,Z=0),Scale3D=(X=1,Y=1,Z=1))),(Item=(Type=Bone,Name=\"dyn_hair_15\"),Transform=(Rotation=(X=0,Y=0.342,Z=0,W=0.9397),Translation=(X=0,Y=0,Z=0),Scale3D=(X=1,Y ...(truncated)"
    WeightMinimum=0
    Weight=@n1.OutputParam
   )
  ]}
]
hierarchy [
 root {loc=<0 0 0>, rot=<0 0 0 1>}
 pelvis {parent=root, loc=<0 -0.7071 93.1191>, rot=<0.0127 0.707 -0.0127 -0.707>}
 spine_01 {parent=pelvis, loc=<6.7494 -0.6388 0>, rot=<0 0 0.034 -0.9994>}
 spine_02 {parent=spine_01, loc=<7.2518 -0.035 0>, rot=<0 0 0.0238 -0.9997>}
 spine_03 {parent=spine_02, loc=<6.0104 0 0>, rot=<0 0 0.0149 -0.9999>}
 spine_04 {parent=spine_03, loc=<8.1024 0 0>, rot=<0 0 -0.0132 -0.9999>}
 spine_05 {parent=spine_04, loc=<10.0146 0 0>, rot=<0 0 -0.0015 -1>}
 neck_01 {parent=spine_05, loc=<11.3138 0.1529 0>, rot=<0 0 -0.2003 -0.9797>}
 neck_02 {parent=neck_01, loc=<4.5627 0 0>, rot=<0 0 0.0504 -0.9987>}
 head {parent=neck_02, loc=<5.3914 0 0>, rot=<0 0 0.115 -0.9934>}
 dyn_hair_31 {parent=head, loc=<2.9433 -1.7909 7.3602>, rot=<0.0265 -0.0535 0.9976 -0.0338>}
 dyn_hair_30 {parent=head, loc=<1.9956 -3.7253 6.3746>, rot=<0.0297 -0.1391 0.9889 -0.0423>}
 dyn_hair_29 {parent=head, loc=<4.8695 -1.9507 7.7997>, rot=<0.104 -0.1878 0.9493 -0.2297>}
 dyn_hair_28 {parent=head, loc=<4.7334 -3.9462 7.4091>, rot=<0.0313 -0.3137 0.9191 -0.2363>}
 dyn_hair_27 {parent=head, loc=<5.8579 -5.6555 5.7723>, rot=<0.0838 -0.5133 0.8245 -0.2231>}
 dyn_hair_26 {parent=head, loc=<2.1504 -6.42 4.4263>, rot=<0.1835 -0.5504 0.8114 -0.0703>}
 dyn_hair_25 {parent=head, loc=<1.7803 -7.1184 2.3822>, rot=<0.157 -0.6477 0.721 -0.1895>}
 dyn_hair_24 {parent=head, loc=<0.471 -6.9927 1.3633>, rot=<0.1047 -0.6597 0.7305 -0.1423>}
 dyn_hair_23 {parent=head, loc=<0.3155 -6.3912 -0.5923>, rot=<0.0489 -0.7679 0.6343 -0.0754>}
 dyn_hair_22 {parent=head, loc=<0.555 -5.7982 -2.3315>, rot=<0.0563 -0.8285 0.5527 -0.0701>}
 dyn_hair_21 {parent=head, loc=<0.8716 -5.0059 -4.3555>, rot=<0.0003 -0.8779 0.4787 0.0123>}
 dyn_hair_20 {parent=head, loc=<2.7179 -2.0701 -6.9649>, rot=<0.0341 -0.959 0.2501 -0.129>}
 dyn_hair_19 {parent=head, loc=<3.1847 -3.9866 -6.557>, rot=<-0.003 -0.9546 0.2655 -0.1349>}
 dyn_hair_18 {parent=head, loc=<2.1985 -4.7957 -5.4283>, rot=<0.12 -0.8474 0.4986 -0.1373>}
 dyn_hair_17 {parent=head, loc=<1.8206 -6.6677 -3.8008>, rot=<-0.036 -0.8567 0.5142 0.0176>}
 dyn_hair_16 {parent=head, loc=<3.636 -7.014 -3.648>, rot=<0.1342 -0.8277 0.5203 -0.1618>}
 dyn_hair_15 {parent=head, loc=<3.2752 -7.5536 -1.084>, rot=<0.0048 -0.8008 0.5986 -0.0157>}
 dyn_hair_14 {parent=head, loc=<4.5929 -7.8592 0.9214>, rot=<0.0921 -0.6937 0.7119 -0.0587>}
 dyn_hair_13 {parent=head, loc=<6.0443 -7.1221 3.2163>, rot=<0.141 -0.6332 0.753 -0.1103>}
 dyn_hair_12 {parent=head, loc=<7.9228 -7.8742 2.0281>, rot=<0.0795 -0.6401 0.7619 -0.0586>}
 dyn_hair_11 {parent=head, loc=<7.0786 -8.2034 -0.2357>, rot=<0.0564 -0.6849 0.7207 -0.0908>}
 dyn_hair_10 {parent=head, loc=<6.1683 -8.0913 -2.9461>, rot=<0.0667 -0.7648 0.6353 -0.0836>}
 dyn_hair_09 {parent=head, loc=<6.4485 -5.5373 -5.6429>, rot=<0.3226 -0.8998 0.281 -0.0852>}
 dyn_hair_08 {parent=head, loc=<5.1931 -2.6393 -7.1215>, rot=<0.6147 -0.7727 0.153 -0.0401>}
 dyn_hair_07 {parent=head, loc=<8.5866 -0.3247 -7.595>, rot=<0.629 -0.7635 0.0373 -0.1413>}
 dyn_hair_06 {parent=head, loc=<10.6953 -7.3345 -0.5367>, rot=<0.3429 -0.605 0.6631 -0.2769>}
 dyn_hair_05 {parent=head, loc=<9.6119 -6.8186 4.3997>, rot=<0.2029 -0.3058 0.8526 -0.3721>}
 dyn_hair_04 {parent=head, loc=<9.8389 -3.5972 6.3545>, rot=<0.0283 -0.3594 0.742 -0.5652>}
 dyn_hair_03 {parent=head, loc=<13.1608 2.9131 5.7138>, rot=<0.3171 -0.0579 0.6331 -0.7038>}
 dyn_hair_02 {parent=head, loc=<15.3355 5.2666 -2.1901>, rot=<0.6896 -0.3491 0.5016 -0.3885>}
 dyn_hair_01 {parent=head, loc=<14.731 2.7693 -4.7835>, rot=<0.7039 -0.5405 0.3678 -0.2777>}
 clavicle_l {parent=spine_05, loc=<4.3575 -0.1466 -2.3223>, rot=<0.0465 -0.7289 -0.009 -0.6829>}
 clavicle_r {parent=spine_05, loc=<4.3575 -0.1466 2.3223>, rot=<0.7289 0.0465 -0.6829 0.009>}
]
